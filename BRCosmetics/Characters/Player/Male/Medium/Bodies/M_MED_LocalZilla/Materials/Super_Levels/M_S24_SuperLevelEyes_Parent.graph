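# Parent material for the eyes of the LocalZilla character's S24 Super Level styles, built on the shared MF_eyes function.
# Desaturated eye base color blends two tint colors, plus a Fresnel-masked sky light environment reflection added directly into Base Color.
# A second Fresnel, using vertex normals, blends two glow colors and intensities into Emissive, while parameters scale metallic, specular and roughness.
# The result passes through MF_TechArtOverrides and the character PostFX function.
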Material /BRCosmetics/Characters/Player/Male/Medium/Bodies/M_MED_LocalZilla/Materials/Super_Levels/M_S24_SuperLevelEyes_Parent
graph {
 properties {
  bUseMaterialAttributes=true
  bUsedWithMeshDeformer=true
  bUsedWithSkeletalMesh=true
  pixelDepthOffsetMode=PDOM_Legacy
  shadingModel=MSM_FromMaterialExpression
  usedShadingModels=Default Lit
 }
 nodes [
  n1 = GetMaterialAttributes(None=MaterialFunctionCall() {materialFunction=/Game/Characters/Shared/FNEye/Materials/MF_eyes}.Result) {attributeGetTypes [
    #1
    #2
    #3
    #4
    #5
   ]}
  n2 = Multiply(
   A=SkyLightEnvMapSample(Direction=VertexNormalWS())
   B=Fresnel(ExponentIn=ScalarParameter() {defaultValue=2, group=Super Level, parameterName=Fresnel Exponent Ref}) {baseReflectFraction=0}
  )
  n3 = ScalarParameter() {defaultValue=1, group=Super Level, parameterName=ReflectionAmount}
  n4 = VectorParameter() {group=Super Level, parameterName=BaseColor1}
  n5 = VectorParameter() {group=Super Level, parameterName=BaseColor2}
  n6 = ComponentMask(None=Desaturation(None=@n1.Base Color, Fraction=Constant() {r=1})) {r=true}
  n7 = Fresnel(ExponentIn=ScalarParameter() {defaultValue=0.6, group=Super Level, parameterName=Fresnel Exponent}, Normal=VertexNormalWS()) {baseReflectFraction=0}
  n8 = LinearInterpolate(
   A=VectorParameter() {group=Super Level, parameterName=Color 1, defaultValue=<rgba 0.3506 0 0.3841 1>}.RGB
   B=VectorParameter() {group=Super Level, parameterName=Color 2, defaultValue=<rgba 0.03 0.1186 0.9 1>}.RGB
   Alpha=@n7
  )
  n9 = LinearInterpolate(
   A=ScalarParameter() {defaultValue=1, group=Super Level, parameterName=Glow Intensity 1}
   B=ScalarParameter() {defaultValue=2, group=Super Level, parameterName=Glow Intensity 2}
   Alpha=@n7
  )
  n10 = SetMaterialAttributes!(
   MaterialAttributes=@n1.MaterialAttributes
   Base Color=Add(A=Multiply(A=@n2, B=@n3), B=LinearInterpolate(A=@n4.RGB, B=@n5.RGB, Alpha=@n6))
   Normal
   Emissive Color=Add(A=@n1.Emissive Color, B=Multiply(A=@n8, B=@n9))
   Metallic=Multiply(A=@n1.Metallic, B=ScalarParameter() {defaultValue=1, group=Super Level, parameterName=Metallic, sortPriority=1000})
   Specular=Multiply(A=@n1.Specular, B=ScalarParameter() {defaultValue=1, group=Super Level, parameterName=Specular, sortPriority=1500})
   Roughness=Multiply(A=@n1.Roughness, B=ScalarParameter() {defaultValue=0.05, group=Super Level, parameterName=Roughness, sortPriority=2000})
  ) {attributeSetTypes [
    #1
    #6
    #2
    #3
    #4
    #5
   ]}
  MaterialFunctionCall(Material=MaterialFunctionCall(In=@n10) {materialFunction=&MF_TechArtOverrides}.Result) {materialFunction=&BaseFN_CharacterFunction_PostFX}
 ]
}
