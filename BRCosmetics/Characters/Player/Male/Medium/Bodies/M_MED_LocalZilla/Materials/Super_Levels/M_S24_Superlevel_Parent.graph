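# Parent character material for the S24 Super Level looks, stored under the LocalZilla body, using Material Attributes and an expression-driven shading model.
# Base PreFX output passes through CharacterSkinFX, which the "Disable SkinFX" static switch can bypass, into the MFC_S24_SuperLevel function.
# That result runs through shared PostFX and cinematic post functions.
# Named reroutes share AO, Gmap, their product, skin mask and pre-skinned position from PreFX.
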Material /BRCosmetics/Characters/Player/Male/Medium/Bodies/M_MED_LocalZilla/Materials/Super_Levels/M_S24_Superlevel_Parent
graph {
 properties {
  bAutomaticallySetUsageInEditor=false
  bUseMaterialAttributes=true
  bUsedWithMeshDeformer=true
  bUsedWithMorphTargets=true
  bUsedWithNanite=true
  bUsedWithSkeletalMesh=true
  pixelDepthOffsetMode=PDOM_Legacy
  refractionMethod=RM_IndexOfRefraction
  shadingModel=MSM_FromMaterialExpression
  subsurfaceProfile=&SS_HeroSkin_02
  usedShadingModels=Default Lit | Subsurface | Clear Coat | Subsurface Profile | Cloth | Eye
 }
 nodes [
  n1 = MaterialFunctionCall() {materialFunction=&BaseFN_CharacterFunction_PreFX}
  n2 = MaterialFunctionCall(
   PreskinnedPosition=@n1.Pre skinned Local Position
   Gmap=@n1.Pre skinned Local Position
   MaterialAttributes=@n1.Pre skinned Local Position
   AO=@n1.Pre skinned Local Position
  ) {materialFunction=&CharacterSkinFX}
  n3 = MaterialFunctionCall(
   AO=@n1.AO
   Distance Blend=@n1.AO
   Fresnel=@n1.AO
   Gmap=@n1.AO
   Pre Skinned Local Position=@n1.AO
   Material=MaterialFunctionCall!(
    In=StaticSwitchParameter(True=@n1.Result, False=@n2.Result) {parameterName=Disable SkinFX}
    Preskinned Position=NamedRerouteUsage()
    AO=NamedRerouteUsage()
    Gmap=NamedRerouteUsage()
    Skin=NamedRerouteUsage()
    Toon Shading
    Inkline
   ) {materialFunction=/Game/Characters/Player/Male/Medium/Bodies/M_MED_LocalZilla/Materials/Super_Levels/SL24_Functions/MFC_S24_SuperLevel}.Result
  ) {materialFunction=&BaseFN_CharacterFunction_PostFX}
  MaterialFunctionCall(In=@n3.Result) {materialFunction=&CINE_MF_CharPostFX}
  NamedRerouteDeclaration(None=Multiply(
    A=NamedRerouteDeclaration(None=@n1.AO) {name=AO, nodeColor=<rgba 0.8984 1 0 1>}
    B=NamedRerouteDeclaration(None=@n1.Gmap) {name=Gmap, nodeColor=<rgba 0.8984 1 0 1>}
   )) {name=GAO, nodeColor=<rgba 0.8984 1 0 1>}
  NamedRerouteDeclaration(None=@n1.Skin) {name=Skin Mask, nodeColor=<rgba 0.9948 1 0.9896 1>}
  NamedRerouteDeclaration(None=@n1.Pre skinned Local Position) {name=Preskinned Local Position, nodeColor=<rgba 1 0.1974 0.1874 1>}
  StaticBool() {value=true}
 ]
}
